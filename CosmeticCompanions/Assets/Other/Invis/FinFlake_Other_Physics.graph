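# FinFlake_Other_Physics is a physics asset in the invisible "Other" cosmetic companion assets, apparently for FinFlake.
# It defines a single body, body_01, with one capsule that has a near-zero length and a large radius, so it acts almost as a sphere.
# The capsule is rotated and slightly offset from the bone origin, and there are no constraints.
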PhysicsAsset /CosmeticCompanions/Assets/Other/Invis/FinFlake_Other_Physics
bodies {body_01 [{shapeType=Capsule, length=0.281, radius=37.1296, center=<-0.1621 -6.5452 0>, rotation=<rot 0 -90 0>}]}
constraints []
