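# Physics asset for the HoneySting invisible cosmetic companion mesh, in the CosmeticCompanions Other/Invis folder.
# It holds a single body on the fish_01 bone, a root capsule offset from the bone origin and short enough to be nearly spherical.
# There are no constraints, so it defines one simple collision volume with no articulated ragdoll chain.
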PhysicsAsset /CosmeticCompanions/Assets/Other/Invis/HoneySting_Other_invis_PhysicsAsset
bodies {fish_01 [{shapeType=Capsule, length=1.5, radius=28.5, shapeName=root_capsule, center=<0 -10.2302 2.9052>}]}
constraints []
